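; Rocket projectile for the Ashton pack Milo launcher, exploding on hit or when its 15-second fuse timer ends.
; On spawn it attaches a trail and passes the weapon muzzle position to Niagara, then spins the mesh each tick.
; Touching "NoStick" components doubles velocity; other touches set stop-on-hit.
; Explosion cosmetics include water-aware sound, proximity post-process fade, camera shake, force feedback and chest effects on hit players.
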(event EventBeginPlay
  (Development|PrintString "Spawned")
  ;[1 1 1] Begin Play
  (Utilities|FlowControl|Sequence
    (:then_0
      (Networking|CanExecuteCosmeticEvents
        (:True
          (bind _returnvalue (Effects|SpawnFXSystemAttached (Variables|Default|GetRocket%20Trail%20Template) (Variables|Projectile_Rocket|GetMesh) "None" "-240.000000,0.000000,0.000000" (Math|Rotator|MakeRotator (Math|Random|RandomFloatinRange 0.0 360))))
          (bind _output_get (Variables|Effects|SetRocketTrailPS _returnvalue))
          (bind _asniagara_particle_system_component (Utilities|Casting|CastToNiagaraComponent _output_get)
            (:then
              (Utilities|FlowControl|Sequence
                (:then_0
                  _)
                (:then_1
                  (bind _returnvalue_2 (CallFunction|GetOwnerWeapon))
                  (Niagara|SetNiagaraVariable(Position) _asniagara_particle_system_component "MuzzlePosWS" (Transformation|GetSocketLocation (Visual|GetWeaponMesh _returnvalue_2) "Muzzle"))
                  (Development|PrintString "Hello"))))
            (:CastFailed
              _))
          (Rendering|SetRenderCustomDepth _output_get true))))
    (:then_1
      (Utilities|Time|SetTimerbyFunctionName self "FuseTimerMax" 15))))

(event EventTick
  (params (DeltaSeconds :type Float%20%28single-precision%29))
  (bind _rockettrailps (Variables|Effects|GetRocketTrailPS))
  ;[1 1 1] On Tick, spin the rocket and particle trail
  (Variables|Default|SetTick%20Delta DeltaSeconds)
  (Networking|CanExecuteCosmeticEvents
    (:True
      (if (Utilities|IsValid _rockettrailps :!class /Script/Engine.KismetSystemLibrary)
        (if (Components|Activation|IsActive _rockettrailps)
          (Variables|Effects|SetRocketSpin (Math|Rotator|MakeRotator (* (Variables|Default|GetTick%20Delta) (Variables|Effects|GetRocketSpinRate))))
          (Transformation|AddLocalRotation (Variables|Projectile_Rocket|GetMesh) (Variables|Effects|GetRocketSpin) false false))))))

(event EventOnTouched
  (params (OtherActor :type Actor%20Object%20Reference) (OtherComp :type Primitive%20Component%20Object%20Reference) (HitResult :type Hit%20Result%20Structure%20%28by%20ref%29) (bIsOverlap :type Boolean))
  (bind (_bblockinghit _binitialoverlap _time _distance _location _impactpoint _normal _impactnormal _physmat _hitactor _hitcomponent) (Collision|BreakHitResult HitResult))
  ;[1 1 1] EVENT -> On Touched - Bounce off "NoBounce"
  (if (Components|ComponentHasTag _hitcomponent "NoStick")
    (Variables|/Script/Engine.MovementComponent|SetVelocity (* (Variables|/Script/Engine.MovementComponent|GetVelocity (Variables|Movement|GetProjectileMovementComponent)) 2.0) (Variables|Movement|GetProjectileMovementComponent))
    (else
      (Variables|Projectile|SetbStopSimulatingOnHit true))))

(event Collision|EventHit
  (params (MyComp :type Primitive%20Component%20Object%20Reference) (Other :type Actor%20Object%20Reference) (OtherComp :type Primitive%20Component%20Object%20Reference) (bSelfMoved :type Boolean) (HitLocation :type Vector :default "0, 0, 0") (HitNormal :type Vector :default "0, 0, 0") (NormalImpulse :type Vector :default "0, 0, 0") (Hit :type Hit%20Result%20Structure%20%28by%20ref%29))
  (bind (_bblockinghit _binitialoverlap _time _distance _location _impactpoint _normal _impactnormal _physmat) (Collision|BreakHitResult Hit))
  ;[1 1 1] EVENT -> On Hit
  (Variables|Effects|SetDecalLocation _impactpoint)
  (Variables|Effects|SetStopLocation (+ _impactpoint (* _impactnormal (Variables|Effects|GetExplosion%20Impact%20Offset))))
  (Variables|Effects|SetStopRotZ (Math|Rotator|MakeRotfromZ _impactnormal))
  (Variables|Default|SetSurfaceType (Variables|/Script/PhysicsCore.PhysicalMaterial|GetSurfaceType _physmat))
  (if (not (Delivery|IsBeingKilled))
    (Delivery|DoExplosion)))

(event Custom|FuseTimerMax
  ;[1 1 1] EVENT -> On Hit
  (Variables|Effects|SetTimerMaxReached? true)
  (Variables|Effects|SetStopLocation (Transformation|GetWorldLocation (Variables|Projectile_Rocket|GetMesh)))
  (if (not (Delivery|IsBeingKilled))
    (Delivery|DoExplosion)))

(event EventOnExploded
  (params (HitActors :type Array%20of%20Actor%20Object%20References) (HitResults :type Array%20of%20Hit%20Result%20Structures))
  (bind _stoplocation (Variables|Effects|GetStopLocation))
  (bind _stoplocation_2 (Variables|Effects|GetStopLocation))
  (bind _explosion_generic_ps (Variables|Effects|GetExplosion_Generic_PS))
  (bind _stoprotz (Variables|Effects|GetStopRotZ))
  ;[1 1 1] EVENT -> On Exploded - Impact and Explosion Stuff
  (Networking|CanExecuteCosmeticEvents
    (:True
      (Utilities|FlowControl|Sequence
        (:then_0
          (Development|SetHiddeninGame (Variables|Projectile_Rocket|GetMesh) true false)
          (Components|Activation|SetActive (Variables|Effects|GetRocketTrailPS) false false)
          ;;[1 1 1] SingleLineCheckForWorldObjectsSurfaceType
          (if (Utilities|Enum|Equal(Enum) (Variables|Default|GetSurfaceType) "SurfaceType8")
            (bind _returnvalue (Effects|SpawnEmitteratLocation _explosion_generic_ps _stoplocation _stoprotz))
            (Rendering|SetRenderCustomDepth _returnvalue true)
            (Audio|PlaySoundatLocation (Variables|Sounds|GetRocket_Projectile_Explosion_Water_Sound) (Variables|Effects|GetStopLocation))
            (else
              (bind _returnvalue_2 (Effects|SpawnEmitteratLocation _explosion_generic_ps _stoplocation _stoprotz))
              (Rendering|SetRenderCustomDepth _returnvalue_2 true)
              (Audio|PlaySoundatLocation (Variables|Sounds|GetRocket_Projectile_Explosion_Sound) (Variables|Effects|GetStopLocation))))
          (Rendering|SetVisibility (Variables|Projectile_Rocket|GetMesh) false false)
          ;[1 1 1] Shake that camera / Post Process
          (Utilities|FlowControl|Sequence
            (:then_0
              (if (<= (Math|Vector|VectorLength (- (Variables|Effects|GetStopLocation) (Transformation|GetWorldLocation (Variables|/Script/Engine.PlayerCameraManager|GetTransformComponent (Game|GetPlayerCameraManager 0))))) (Variables|/Script/Engine.PostProcessComponent|GetBlendRadius (Variables|Default|GetPP_Explode)))
                (bind (_direction _ppfade) (Timeline|PPFader PlayFromStart)
                  (:Update
                    (Variables|/Script/Engine.PostProcessComponent|SetBlendWeight _ppfade (Variables|Default|GetPP_Explode))))))
            (:then_1
              (Camera|PlayWorldCameraShake "/Game/Weapons/FORT_RocketLaunchers/Blueprints/B_RocketLauncherCameraShake.B_RocketLauncherCameraShake_C" _stoplocation_2 400 850)
              (Audio|Components|Audio|FadeOut (Variables|Default|GetRocket_AudioComponent) 1 0.0 "Linear" :!class /Script/Engine.AudioComponent)
              (Fortnite|PlayLocalForceFeedbackatLocationMulti (Variables|Default|GetExplosionForceFeedbackNear) (Variables|Default|GetExplosionForceFeedbackFar) _stoplocation_2 400 850 "Rocket"))))
        (:then_1
          ;;[1 1 1] Hit Player FX
          (for _array_element HitActors
            (bind _asplayer_pawn_athena (Utilities|Casting|CastToPlayerPawn_Athena _array_element))
            (Effects|SpawnEmitterAttached "/Figment_Martin/Athena/Items/Gameplay/BackPacks/Ashton/Milo/FX/Secondary/P_Assaul_Ashton_Secondary_Milo_Impact.P_Assaul_Ashton_Secondary_Milo_Impact" (Variables|/Script/Engine.Character|GetMesh _asplayer_pawn_athena) "FX_Chest")))))))

(event EventDestroyed
  ;[1 1 1] EVENT -> Destroyed
  (Networking|CanExecuteCosmeticEvents
    (:True
      (Components|Activation|SetActive (Variables|Effects|GetRocketTrailPS) false false))))

(fn ConstructionScript
  (Effects|Components|ParticleSystem|SetTemplate (Variables|Default|GetRibbon_Trail_PSC) (Variables|Effects|GetRibbon_Trail_PS))
  (Effects|Components|ParticleSystem|SetTemplate (Variables|Default|GetExplosion_Generic_PSC) (Variables|Effects|GetExplosion_Generic_PS))
  (Effects|Components|ParticleSystem|SetTemplate (Variables|Default|GetExplosion_Flesh_Damage_PSC) (Variables|Effects|GetExplosion_Flesh_Damage_PS)))
